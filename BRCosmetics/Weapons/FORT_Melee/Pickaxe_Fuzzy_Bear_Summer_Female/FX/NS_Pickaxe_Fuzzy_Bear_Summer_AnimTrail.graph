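# Niagara anim-trail effect that draws three claw swipe trails for the Fuzzy Bear Summer (Female) pickaxe.
# Claw and Soft emitters each spawn three trails from bone/socket pairs, narrowing and fading over User.Lifetime, with Soft adding a color curve.
# Ribbon facing interpolates toward the owner's X axis; the bIsOffhand width switch yields -1 either way.
# Six single-trail emitters (Claw01–03, Soft01–03) using position offsets are disabled.
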NiagaraSystem /BRCosmetics/Weapons/FORT_Melee/Pickaxe_Fuzzy_Bear_Summer_Female/FX/NS_Pickaxe_Fuzzy_Bear_Summer_AnimTrail
dependencies {
 usedDataInterfaces [&NiagaraDataInterfaceCurve, &NiagaraDataInterfaceColorCurve]
 usedDynamicInputs [&MakeFloatFromVector]
 usedModules [&SystemState]
}
emitters {
 Claw {
  emitterUpdateScript [
   &EmitterState()
   &FN_SpawnAnimTrail_v2(Number of Anim Trails=3)
  ]
  particleSpawnScript [
   /Game/Effects/Niagara/Modules/AnimTrail/FN_InitializeAnimTrail/FN_InitializeAnimTrailBoneIndecies(Second Bone or Socket Index=1)
   &FN_InitializeAnimTrailBoneIndecies(AnimTrail Selection=1, First Bone or Socket Index=2, Second Bone or Socket Index=3) {module=FN_InitializeAnimTrailBoneIndecies001}
   &FN_InitializeAnimTrailBoneIndecies(AnimTrail Selection=2, First Bone or Socket Index=4, Second Bone or Socket Index=5) {module=FN_InitializeAnimTrailBoneIndecies002}
   &FN_SetSpawnGroup(Normalize Spawn Group=true, Number of Spawn Groups=3, Set Ribbon ID and Ribbon Link Order=true) {enabled=false}
   /SetVariables(Particles.FacingVec:Vector3f=$Engine.Owner.SystemXAxis) {isSetParameters=true, module=SetVariables_#1}
   /Game/Effects/Niagara/Modules/AnimTrail/FN_InitializeAnimTrail/FN_InitializeAnimTrail(
    AnimTrailWidth:NiagaraFloat=&Multiply_Float(A=1, B=&MakeCustomFloatFromBool(Bool=$User.bIsOffhand, True Float=-1, False Float=-1))
    Color=<rgba 0.3 0.3 0.5 0.3>
    Lifetime:NiagaraFloat=$User.Lifetime
   )
  ]
  particleUpdateScript [
   &ScaleRibbonWidth(Ribbon Width Scale=&FloatFromCurve(FloatCurve=DataInterface() {propertyValues {Curve {keys [
         {arriveTangent=0.1159, interpMode=RCIM_Cubic, leaveTangent=0.1159, tangentMode=RCTM_User, time=0, value=1}
         {arriveTangent=-1, interpMode=RCIM_Cubic, leaveTangent=-1, time=0.5, value=0.15}
         {interpMode=RCIM_Cubic, leaveTangent=-0.9664, time=1, value=0}
        ]}}}))
   &ScaleColor(
    Scale Alpha=&FloatFromCurve(FloatCurve=DataInterface() {propertyValues {Curve {keys [<at 0 0.25>,<at 1 0>]}}})
    Scale RGB=&VectorFromFloat(Value=&FloatFromCurve(FloatCurve=DataInterface() {propertyValues {Curve {keys [<at 0 1 RCIM_Cubic>, {arriveTangent=-1, interpMode=RCIM_Cubic, leaveTangent=-1, time=0.2, value=2}, <at 1 0 RCIM_Cubic>]}}}))
    ScaleA=true
    ScaleRGB=true
   ) {enabled=false}
   &ScaleColor(
    Scale Alpha=&FloatFromCurve(FloatCurve=DataInterface() {propertyValues {Curve {keys [<at 0 0.25>,<at 1 0>]}}})
    ScaleA=true
    ScaleRGB=true
   ) {module=ScaleColor001}
   &InterpolateOverTime(
    Initial Vector Value:Vector3f=$Particles.FacingVec
    Rate Of Change=10
    Target Vector Value:Vector3f=$Engine.Owner.SystemXAxis
    Variable Type=::NewEnumerator2
   )
   /SetVariables(Particles.RibbonFacing=$Particles.FacingVec) {isSetParameters=true, module=SetVariables_#2}
   &FN_UpdateAnimTrail_v2()
  ]
  rendererClasses [&NiagaraRibbonRendererProperties, &NiagaraSpriteRendererProperties]
  renderers [
   {rendererClass=&NiagaraRibbonRendererProperties}
   {rendererClass=&NiagaraSpriteRendererProperties, rendererIndex=1}
   {rendererClass=&NiagaraRibbonRendererProperties, rendererIndex=2}
  ]
 }
 Claw01 {
  bEnabled=false
  emitterUpdateScript [
   &EmitterState(Life Cycle Mode=::NewEnumerator1, Loop Behavior=::NewEnumerator1, Loop Duration=$User.Loop duration, UseLoopDelay=false)
   &FN_SpawnAnimTrail_v2()
  ]
  particleSpawnScript [
   &FN_InitializeAnimTrailBoneIndecies(Second Bone or Socket Index=1)
   &FN_InitializeAnimTrail(
    AnimTrailWidth:NiagaraFloat=$User.AnimTrailWidth
    Color=<rgba 0.3 0.3 0.5 0.3>
    Lifetime:NiagaraFloat=$User.Lifetime
    Position Offset:Vector3f=<0 2 0>
   )
  ]
  particleUpdateScript [
   &ScaleRibbonWidth(Ribbon Width Scale=&FloatFromCurve(FloatCurve=DataInterface() {propertyValues {Curve {keys [
         {arriveTangent=-1.2579, interpMode=RCIM_Cubic, time=0, value=1}
         {arriveTangent=-0.9664, interpMode=RCIM_Cubic, leaveTangent=-0.9664, tangentMode=RCTM_User, time=1, value=0}
        ]}}}))
   &ScaleColor(
    Scale Alpha=&FloatFromCurve(FloatCurve=DataInterface() {propertyValues {Curve {keys [<at 0 0.25>,<at 1 0>]}}})
    ScaleA=true
    ScaleRGB=true
   )
   &FN_UpdateAnimTrail_v2()
  ]
  rendererClasses [&NiagaraRibbonRendererProperties]
  renderers [{rendererClass=&NiagaraRibbonRendererProperties}]
 }
 Claw02 {
  bEnabled=false
  emitterUpdateScript [
   &EmitterState(Life Cycle Mode=::NewEnumerator1, Loop Behavior=::NewEnumerator1, Loop Duration=$User.Loop duration, UseLoopDelay=false)
   &FN_SpawnAnimTrail_v2()
  ]
  particleSpawnScript [
   &FN_InitializeAnimTrailBoneIndecies(Second Bone or Socket Index=1)
   &FN_InitializeAnimTrail(
    AnimTrailWidth:NiagaraFloat=$User.AnimTrailWidth
    Color=<rgba 0.3 0.3 0.5 0.3>
    Lifetime:NiagaraFloat=$User.Lifetime
    Position Offset:Vector3f=<0 0 -8>
   )
  ]
  particleUpdateScript [
   &ScaleRibbonWidth(Ribbon Width Scale=&FloatFromCurve(FloatCurve=DataInterface() {propertyValues {Curve {keys [
         {arriveTangent=-1.2579, interpMode=RCIM_Cubic, time=0, value=1}
         {arriveTangent=-0.9664, interpMode=RCIM_Cubic, leaveTangent=-0.9664, tangentMode=RCTM_User, time=1, value=0}
        ]}}}))
   &ScaleColor(
    Scale Alpha=&FloatFromCurve(FloatCurve=DataInterface() {propertyValues {Curve {keys [<at 0 0.25>,<at 1 0>]}}})
    ScaleA=true
    ScaleRGB=true
   )
   &FN_UpdateAnimTrail_v2()
  ]
  rendererClasses [&NiagaraRibbonRendererProperties]
  renderers [{rendererClass=&NiagaraRibbonRendererProperties}]
 }
 Claw03 {
  bEnabled=false
  emitterUpdateScript [
   &EmitterState(Life Cycle Mode=::NewEnumerator1, Loop Behavior=::NewEnumerator1, Loop Duration=$User.Loop duration, UseLoopDelay=false)
   &FN_SpawnAnimTrail_v2()
  ]
  particleSpawnScript [
   &FN_InitializeAnimTrailBoneIndecies(Second Bone or Socket Index=1)
   &FN_InitializeAnimTrail(
    AnimTrailWidth:NiagaraFloat=$User.AnimTrailWidth
    Color=<rgba 0.3 0.3 0.5 0.3>
    Lifetime:NiagaraFloat=$User.Lifetime
    Position Offset:Vector3f=<0 0 8>
   )
  ]
  particleUpdateScript [
   &ScaleRibbonWidth(Ribbon Width Scale=&FloatFromCurve(FloatCurve=DataInterface() {propertyValues {Curve {keys [
         {arriveTangent=-1.2579, interpMode=RCIM_Cubic, time=0, value=1}
         {arriveTangent=-0.9664, interpMode=RCIM_Cubic, leaveTangent=-0.9664, tangentMode=RCTM_User, time=1, value=0}
        ]}}}))
   &ScaleColor(
    Scale Alpha=&FloatFromCurve(FloatCurve=DataInterface() {propertyValues {Curve {keys [<at 0 0.25>,<at 1 0>]}}})
    ScaleA=true
    ScaleRGB=true
   )
   &FN_UpdateAnimTrail_v2()
  ]
  rendererClasses [&NiagaraRibbonRendererProperties]
  renderers [{rendererClass=&NiagaraRibbonRendererProperties}]
 }
 Soft {
  emitterUpdateScript [
   &EmitterState()
   &FN_SpawnAnimTrail_v2(Number of Anim Trails=3)
  ]
  particleSpawnScript [
   &FN_InitializeAnimTrailBoneIndecies(Second Bone or Socket Index=1)
   &FN_InitializeAnimTrailBoneIndecies(AnimTrail Selection=1, First Bone or Socket Index=2, Second Bone or Socket Index=3) {module=FN_InitializeAnimTrailBoneIndecies001}
   &FN_InitializeAnimTrailBoneIndecies(AnimTrail Selection=2, First Bone or Socket Index=4, Second Bone or Socket Index=5) {module=FN_InitializeAnimTrailBoneIndecies002}
   &FN_SetSpawnGroup(Normalize Spawn Group=true, Number of Spawn Groups=3, Set Ribbon ID and Ribbon Link Order=true) {enabled=false}
   /SetVariables(Particles.FacingVec:Vector3f=$Engine.Owner.SystemXAxis) {isSetParameters=true, module=SetVariables_#1}
   &FN_InitializeAnimTrail(
    AnimTrailWidth:NiagaraFloat=&Multiply_Float(A=1, B=&MakeCustomFloatFromBool(Bool=$User.bIsOffhand, True Float=-1, False Float=-1))
    Color=<rgba 0.2 0.2 0.5 0.3>
    Lifetime:NiagaraFloat=$User.Lifetime
   )
  ]
  particleUpdateScript [
   &ScaleRibbonWidth(Ribbon Width Scale=&FloatFromCurve(FloatCurve=DataInterface() {propertyValues {Curve {keys [
         {arriveTangent=0.1159, interpMode=RCIM_Cubic, leaveTangent=0.1159, tangentMode=RCTM_User, time=0, value=1}
         {arriveTangent=-1, interpMode=RCIM_Cubic, leaveTangent=-1, time=0.5, value=0.15}
         {interpMode=RCIM_Cubic, leaveTangent=-0.9664, time=1, value=0}
        ]}}}))
   &ScaleColor(
    Scale Alpha=&FloatFromCurve(FloatCurve=DataInterface() {propertyValues {Curve {keys [<at 0 0.25>,<at 1 0>]}}})
    Scale RGB=&VectorFromFloat(Value=&FloatFromCurve(FloatCurve=DataInterface() {propertyValues {Curve {keys [<at 0 1 RCIM_Cubic>, {arriveTangent=-1, interpMode=RCIM_Cubic, leaveTangent=-1, time=0.2, value=2}, <at 1 0 RCIM_Cubic>]}}}))
    ScaleA=true
    ScaleRGB=true
   ) {enabled=false}
   &Color(
    Color=&ColorFromCurve(ColorCurve=DataInterface() {propertyValues {
       AlphaCurve {keys [<at 0 0>, {arriveTangent=0.3073, interpMode=RCIM_Cubic, leaveTangent=0.3073, time=0.15, value=0.3}, <at 0.3254 0.1>, <at 1 0 RCIM_Cubic>]}
       BlueCurve {keys [<at 0.2234 0.5 RCIM_Cubic>,<at 1 0.15>]}
       GreenCurve {keys [<at 0.2234 0.45 RCIM_Cubic>,<at 1 0.12>]}
       RedCurve {keys [<at 0.2234 0.4 RCIM_Cubic>,<at 1 0.1>]}
      }})
    Scale Alpha=1.5
   )
   &InterpolateOverTime(
    Initial Vector Value:Vector3f=$Particles.FacingVec
    Rate Of Change=10
    Target Vector Value:Vector3f=$Engine.Owner.SystemXAxis
    Variable Type=::NewEnumerator2
   )
   /SetVariables(Particles.RibbonFacing=$Particles.FacingVec) {isSetParameters=true, module=SetVariables_#2}
   &FN_UpdateAnimTrail_v2()
  ]
  rendererClasses [&NiagaraRibbonRendererProperties, &NiagaraSpriteRendererProperties]
  renderers [
   {rendererClass=&NiagaraRibbonRendererProperties}
   {rendererClass=&NiagaraSpriteRendererProperties, rendererIndex=1}
   {rendererClass=&NiagaraRibbonRendererProperties, rendererIndex=2}
  ]
 }
 Soft01 {
  bEnabled=false
  emitterUpdateScript [
   &EmitterState(Life Cycle Mode=::NewEnumerator1, Loop Behavior=::NewEnumerator1, Loop Duration=$User.Loop duration, UseLoopDelay=false)
   &FN_SpawnAnimTrail_v2()
  ]
  particleSpawnScript [
   &FN_InitializeAnimTrailBoneIndecies(Second Bone or Socket Index=1)
   &FN_InitializeAnimTrail(AnimTrailWidth=0.8, Color=<rgba 0.2 0.2 0.5 0.3>, Lifetime=0.13, Position Offset:Vector3f=<0 2 0>)
  ]
  particleUpdateScript [
   &ScaleRibbonWidth(Ribbon Width Scale=&FloatFromCurve(FloatCurve=DataInterface() {propertyValues {Curve {keys [
         {arriveTangent=-1.2579, interpMode=RCIM_Cubic, time=0, value=1}
         {arriveTangent=-0.9664, interpMode=RCIM_Cubic, leaveTangent=-0.9664, tangentMode=RCTM_User, time=1, value=0}
        ]}}}))
   &Color(
    Color=&ColorFromCurve(ColorCurve=DataInterface() {propertyValues {
       AlphaCurve {keys [<at 0 0>, {arriveTangent=0.3073, interpMode=RCIM_Cubic, leaveTangent=0.3073, time=0.15, value=0.3}, <at 0.3254 0.1>, <at 1 0 RCIM_Cubic>]}
       BlueCurve {keys [<at 0.2234 0.5 RCIM_Cubic>,<at 1 0.15>]}
       GreenCurve {keys [<at 0.2234 0.45 RCIM_Cubic>,<at 1 0.12>]}
       RedCurve {keys [<at 0.2234 0.4 RCIM_Cubic>,<at 1 0.1>]}
      }})
    Scale Alpha=1.5
   ) {enabled=false}
   &FN_UpdateAnimTrail_v2()
  ]
  rendererClasses [&NiagaraRibbonRendererProperties]
  renderers [{rendererClass=&NiagaraRibbonRendererProperties}]
 }
 Soft02 {
  bEnabled=false
  emitterUpdateScript [
   &EmitterState(Life Cycle Mode=::NewEnumerator1, Loop Behavior=::NewEnumerator1, Loop Duration=$User.Loop duration, UseLoopDelay=false)
   &FN_SpawnAnimTrail_v2()
  ]
  particleSpawnScript [
   &FN_InitializeAnimTrailBoneIndecies(Second Bone or Socket Index=1)
   &FN_InitializeAnimTrail(AnimTrailWidth=0.8, Color=<rgba 0.2 0.2 0.5 0.3>, Lifetime=0.13, Position Offset:Vector3f=<0 0 -8>)
  ]
  particleUpdateScript [
   &ScaleRibbonWidth(Ribbon Width Scale=&FloatFromCurve(FloatCurve=DataInterface() {propertyValues {Curve {keys [
         {arriveTangent=-1.2579, interpMode=RCIM_Cubic, time=0, value=1}
         {arriveTangent=-0.9664, interpMode=RCIM_Cubic, leaveTangent=-0.9664, tangentMode=RCTM_User, time=1, value=0}
        ]}}}))
   &Color(
    Color=&ColorFromCurve(ColorCurve=DataInterface() {propertyValues {
       AlphaCurve {keys [<at 0 0>, {arriveTangent=0.3073, interpMode=RCIM_Cubic, leaveTangent=0.3073, time=0.15, value=0.3}, <at 0.3254 0.1>, <at 1 0 RCIM_Cubic>]}
       BlueCurve {keys [<at 0.2234 0.5 RCIM_Cubic>,<at 0.9918 0.15>]}
       GreenCurve {keys [<at 0.2234 0.45 RCIM_Cubic>,<at 0.9918 0.12>]}
       RedCurve {keys [<at 0.2234 0.4 RCIM_Cubic>,<at 0.9918 0.1>]}
      }})
    Scale Alpha=1.5
   ) {enabled=false}
   &FN_UpdateAnimTrail_v2()
  ]
  rendererClasses [&NiagaraRibbonRendererProperties]
  renderers [{rendererClass=&NiagaraRibbonRendererProperties}]
 }
 Soft03 {
  bEnabled=false
  emitterUpdateScript [
   &EmitterState(Life Cycle Mode=::NewEnumerator1, Loop Behavior=::NewEnumerator1, Loop Duration=$User.Loop duration, UseLoopDelay=false)
   &FN_SpawnAnimTrail_v2()
  ]
  particleSpawnScript [
   &FN_InitializeAnimTrailBoneIndecies(Second Bone or Socket Index=1)
   &FN_InitializeAnimTrail(AnimTrailWidth=0.8, Color=<rgba 0.2 0.2 0.5 0.3>, Lifetime=0.13, Position Offset:Vector3f=<0 0 8>)
  ]
  particleUpdateScript [
   &ScaleRibbonWidth(Ribbon Width Scale=&FloatFromCurve(FloatCurve=DataInterface() {propertyValues {Curve {keys [
         {arriveTangent=-1.2579, interpMode=RCIM_Cubic, time=0, value=1}
         {arriveTangent=-0.9664, interpMode=RCIM_Cubic, leaveTangent=-0.9664, tangentMode=RCTM_User, time=1, value=0}
        ]}}}))
   &Color(
    Color=&ColorFromCurve(ColorCurve=DataInterface() {propertyValues {
       AlphaCurve {keys [<at 0 0>, {arriveTangent=0.3073, interpMode=RCIM_Cubic, leaveTangent=0.3073, time=0.15, value=0.3}, <at 0.3254 0.1>, <at 1 0 RCIM_Cubic>]}
       BlueCurve {keys [<at 0.2234 0.5 RCIM_Cubic>,<at 0.9831 0.15>]}
       GreenCurve {keys [<at 0.2234 0.45 RCIM_Cubic>,<at 0.9831 0.12>]}
       RedCurve {keys [<at 0.2234 0.4 RCIM_Cubic>,<at 0.9831 0.1>]}
      }})
    Scale Alpha=1.5
   ) {enabled=false}
   &FN_UpdateAnimTrail_v2()
  ]
  rendererClasses [&NiagaraRibbonRendererProperties]
  renderers [{rendererClass=&NiagaraRibbonRendererProperties}]
 }
}
summary {
 systemName=NS_Pickaxe_Fuzzy_Bear_Summer_AnimTrail
 emitters [
  {bEnabled=false, emitterName=Claw01, rendererClasses [&NiagaraRibbonRendererProperties]}
  {bEnabled=false, emitterName=Soft01, rendererClasses [&NiagaraRibbonRendererProperties]}
  {bEnabled=false, emitterName=Claw02, rendererClasses [&NiagaraRibbonRendererProperties]}
  {bEnabled=false, emitterName=Claw03, rendererClasses [&NiagaraRibbonRendererProperties]}
  {bEnabled=false, emitterName=Soft02, rendererClasses [&NiagaraRibbonRendererProperties]}
  {bEnabled=false, emitterName=Soft03, rendererClasses [&NiagaraRibbonRendererProperties]}
  {emitterName=Claw, rendererClasses [&NiagaraRibbonRendererProperties, &NiagaraSpriteRendererProperties]}
  {emitterName=Soft, rendererClasses [&NiagaraRibbonRendererProperties, &NiagaraSpriteRendererProperties]}
 ]
 userVariables [
  User.AnimTrailWidth {type=NiagaraFloat, defaultValue=2.3, description=""}
  User.bIsOffhand {type=NiagaraBool}
  User.Lifetime {type=NiagaraFloat, defaultValue=0.2, description=""}
  User.Loop duration {type=NiagaraFloat, defaultValue=0.2, description=""}
  User.LoopDuration {type=NiagaraFloat, defaultValue=0, description=""}
  User.Ribbon camera face {type=Vector3f, description="", defaultValue {struct=Vector3f, value=<0 0 0>}}
  User.RibbonIDOffset {type=NiagaraInt32}
  User.SkeletalMesh {
   type=NiagaraDataInterfaceSkeletalMesh
   defaultValue {
    struct=NiagaraExt_VariableValue_DataInterface
    value {dataInterface=&self:NiagaraDataInterfaceSkeletalMesh_0}
   }
  }
  User.SkeletalMesh_R {
   type=NiagaraDataInterfaceSkeletalMesh
   defaultValue {
    struct=NiagaraExt_VariableValue_DataInterface
    value {dataInterface=&self:NiagaraDataInterfaceSkeletalMesh_1}
   }
  }
 ]
}
