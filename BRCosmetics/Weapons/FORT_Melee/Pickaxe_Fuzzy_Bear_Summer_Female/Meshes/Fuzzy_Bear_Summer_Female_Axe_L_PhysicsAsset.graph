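# Physics asset for the left Fuzzy Bear Summer Female pickaxe, a melee weapon cosmetic, giving its mesh a simplified collision shape.
# It has a single body on the Melee_TwoHanded_Handle bone, built from one offset and rotated capsule.
# It defines no constraints, so there are no jointed or separately simulated parts.
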PhysicsAsset /BRCosmetics/Weapons/FORT_Melee/Pickaxe_Fuzzy_Bear_Summer_Female/Meshes/Fuzzy_Bear_Summer_Female_Axe_L_PhysicsAsset
bodies {Melee_TwoHanded_Handle [{shapeType=Capsule, length=19.0664, radius=9.8096, center=<-0.9856 -0.2676 -2.5138>, rotation=<rot 19.7556 -90 8.2686>}]}
constraints []
