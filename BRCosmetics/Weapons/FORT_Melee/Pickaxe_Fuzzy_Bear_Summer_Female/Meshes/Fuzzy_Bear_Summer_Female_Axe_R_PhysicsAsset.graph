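# Physics asset for the right-hand axe mesh of the female Fuzzy Bear Summer pickaxe cosmetic, a melee weapon.
# It holds a single body on the Melee_TwoHanded_Handle bone, shaped as one rotated capsule slightly offset from the bone origin.
# It defines no constraints, so the weapon has one simple collision volume with no jointed parts.
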PhysicsAsset /BRCosmetics/Weapons/FORT_Melee/Pickaxe_Fuzzy_Bear_Summer_Female/Meshes/Fuzzy_Bear_Summer_Female_Axe_R_PhysicsAsset
bodies {Melee_TwoHanded_Handle [{shapeType=Capsule, length=19.0385, radius=9.3062, center=<0.7115 -0.2653 -2.4198>, rotation=<rot 19.0697 -90 -6.7875>}]}
constraints []
